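# Animation Blueprint for the Meatball motorboat vehicle, animating boost, rocket cooldown, steering, gauges and motor.
# A Boosting state machine, gated by bIsBoostStarting, takes an additive Rocket state machine, then cached steering poses are layered onto the steering wheel branch.
# Bone modifiers then rotate the jet and gauges and add speed-scaled random jiggle to the motor.
# The update event derives jet rotation from clamped steering yaw, interpolating from zero with a fixed timestep.
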AnimBlueprint /Motorboat/Vehicle/Animation/MeatBallVehicle_AnimBP
target_skeleton=/Game/Athena/DrivableVehicles/Meatball/Meshes/Medium_Meatball_Veh_Skeleton
graphs [
 AnimGraph {nodes [
   AnimGraphNode_StateMachine_0 = Boosting() {is=title AnimGraphNode_StateMachine}
   AnimGraphNode_StateMachine_1 = Rocket() {is=title AnimGraphNode_StateMachine}
   n1 = Apply Additive(
    Base=Blend Poses by bool(
     BlendPose_0=@AnimGraphNode_StateMachine_0
     BlendPose_1=Local Space Ref Pose()
     BlendTime_1=0.2
     bActiveValue=Get bIsBoostStarting() -> (bIsBoostStarting:bool)
    ) {blendType=Linear}
    Additive=@AnimGraphNode_StateMachine_1
   ) {alphaBoolBlend {blendOption=Linear}}
   n2 = Blend Poses by bool(
    BlendPose_0=Blend Poses by bool(
     BlendPose_0=Use cached pose 'SteerPose'()
     BlendPose_1=Use cached pose 'SteerPose'()
     BlendTime_0=0.25
     BlendTime_1=0.25
     bActiveValue=Get bIsDrivingFast() -> (bIsDrivingFast:bool)
    ) {blendType=Linear}
    BlendPose_1=Sequence Player 'MeatBallVehicle_Steering_Default'() {sequence=/Game/Animation/Game/Vehicles/MeatBall/MeatBallVehicle_Steering_Default}
    BlendTime_0=0.35
    BlendTime_1=0.35
    bActiveValue=Get bIsMovingForward() -> (bIsMovingForward:bool)
   ) {blendType=Linear}
   n3 = "Transform (Modify) Bone - Bone: jet_direction"(
    ComponentPose=Local To Component(LocalPose=Layered blend per bone(BasePose=@n1, BlendPoses_0=@n2) {layerSetup [{branchFilters [{boneName=steering_wheel}]}]})
    Alpha=1
    Translation="0,0,0"
    Rotation=Get jet_rotation() -> (jet_rotation:Rotator)
    Scale="1,1,1"
   ) {rotationMode=BMM_Replace, rotationSpace=BCS_ParentBoneSpace, alphaBoolBlend {blendOption=Linear}, boneToModify {boneName=jet_direction}}
   n4 = "Transform (Modify) Bone - Bone: guage_speed"!(ComponentPose=@n3, Alpha=1, Rotation=Get SpeedGuageRotator() -> (SpeedGuageRotator:Rotator)) -> (Pose) {
    rotationMode=BMM_Replace
    rotationSpace=BCS_BoneSpace
    scaleSpace=BCS_BoneSpace
    translationSpace=BCS_BoneSpace
    alphaBoolBlend {blendOption=Linear}
    boneToModify {boneName=guage_speed}
   }
   n5 = "Transform (Modify) Bone - Bone: guage_r"!(ComponentPose=@n4, Alpha=1, Rotation=Get BoostPercentRotator() -> (BoostPercentRotator:Rotator)) -> (Pose) {
    rotationMode=BMM_Replace
    rotationSpace=BCS_BoneSpace
    scaleSpace=BCS_BoneSpace
    translationSpace=BCS_BoneSpace
    alphaBoolBlend {blendOption=Linear}
    boneToModify {boneName=guage_r}
   }
   n6 = Get MeatballSpeed() -> (MeatballSpeed:float)
   n7 = "Transform (Modify) Bone - Bone: motor_base"(
    ComponentPose=@n5
    Alpha=@n6
    Translation="0,0,0"
    Rotation=Get RandomJiggleRotatorMotorBase() -> (RandomJiggleRotatorMotorBase:Rotator)
    Scale="1,1,1"
   ) {
    lODThreshold=1
    rotationMode=BMM_Additive
    rotationSpace=BCS_BoneSpace
    alphaBoolBlend {blendOption=Linear}
    alphaScaleBiasClamp {bMapRange=true, scale=0.9, inRange {max=45}, outRange {min=0.2}}
    boneToModify {boneName=motor_base}
   }
   n8 = "Transform (Modify) Bone - Bone: motor_top"(
    ComponentPose=@n7
    Alpha=@n6
    Translation="0,0,0"
    Rotation=Get RandomJiggleRotatorMotorTop() -> (RandomJiggleRotatorMotorTop:Rotator)
    Scale="1,1,1"
   ) {
    lODThreshold=1
    rotationMode=BMM_Additive
    rotationSpace=BCS_BoneSpace
    alphaBoolBlend {blendOption=Linear}
    alphaScaleBiasClamp {bMapRange=true, scale=1.1, inRange {max=65}, outRange {min=0.2}}
    boneToModify {boneName=motor_top}
   }
   Output Pose(Result=Component To Local(ComponentPose=@n8)) {layerGroup=None}
   n9 = Sequence Player 'MeatBallVehicle_Steering_Right'() {sequence=/Game/Animation/Game/Vehicles/MeatBall/MeatBallVehicle_Steering_Right}
   n10 = Sequence Player 'MeatBallVehicle_Steering_Default'() {sequence=&MeatBallVehicle_Steering_Default}
   n11 = Get bIsSteeringRight() -> (bIsSteeringRight:bool)
   Save cached pose 'SteerPose'(Pose=Blend Poses by bool(
     BlendPose_0=Sequence Player 'MeatBallVehicle_Steering_Left'() {sequence=/Game/Animation/Game/Vehicles/MeatBall/MeatBallVehicle_Steering_Left}
     BlendPose_1=Blend Poses by bool(BlendPose_0=@n9, BlendPose_1=@n10, BlendTime_0=0.35, BlendTime_1=0.35, bActiveValue=@n11) {blendType=Linear}
     BlendTime_0=0.35
     BlendTime_1=0.35
     bActiveValue=Get bIsSteeringLeft() -> (bIsSteeringLeft:bool)
    ) {blendType=Linear}) {cacheName=SteerPose}
  ]}
 Boosting {
  entry=Idle
  path=AnimGraph.AnimGraphNode_StateMachine_0.Boosting
  ids=id
  nodes [
   AnimStateNode_1 = BoostReady() {is=title AnimStateNode}
   AnimStateNode_2 = BoostLoop() {is=title AnimStateNode}
   AnimStateNode_3 = BoostOnCD() {is=title AnimStateNode}
   AnimStateNode_4 = Idle()
   AnimStateTransitionNode_0 = AnimStateTransitionNode()
   AnimStateTransitionNode_1 = AnimStateTransitionNode() {crossfadeDuration=0.15}
   AnimStateTransitionNode_2 = AnimStateTransitionNode() {crossfadeDuration=0.033}
   AnimStateTransitionNode_4 = AnimStateTransitionNode() {crossfadeDuration=0.15}
   AnimStateTransitionNode_5 = AnimStateTransitionNode() {crossfadeDuration=0}
   AnimStateTransitionNode_6 = AnimStateTransitionNode() {crossfadeDuration=0.15}
   AnimStateTransitionNode_7 = AnimStateTransitionNode()
  ]
  links [
   {
    dst=BoostOnCD
    rule=AnimGraph.AnimGraphNode_StateMachine_0.Boosting.AnimStateTransitionNode_2.Transition
    src=BoostLoop
    via=AnimStateTransitionNode_2
   }
   {
    dst=Idle
    rule=AnimGraph.AnimGraphNode_StateMachine_0.Boosting.AnimStateTransitionNode_7.Transition
    src=BoostLoop
    via=AnimStateTransitionNode_7
   }
   {
    dst=BoostReady
    rule=AnimGraph.AnimGraphNode_StateMachine_0.Boosting.AnimStateTransitionNode_0.Transition
    src=BoostOnCD
    via=AnimStateTransitionNode_0
   }
   {
    dst=Idle
    rule=AnimGraph.AnimGraphNode_StateMachine_0.Boosting.AnimStateTransitionNode_6.Transition
    src=BoostOnCD
    via=AnimStateTransitionNode_6
   }
   {
    dst=BoostLoop
    rule=AnimGraph.AnimGraphNode_StateMachine_0.Boosting.AnimStateTransitionNode_1.Transition
    src=BoostReady
    via=AnimStateTransitionNode_1
   }
   {
    dst=Idle
    rule=AnimGraph.AnimGraphNode_StateMachine_0.Boosting.AnimStateTransitionNode_5.Transition
    src=BoostReady
    via=AnimStateTransitionNode_5
   }
   {
    dst=BoostReady
    rule=AnimGraph.AnimGraphNode_StateMachine_0.Boosting.AnimStateTransitionNode_4.Transition
    src=Idle
    via=AnimStateTransitionNode_4
   }
  ]
 }
 BoostReady {
  path=AnimGraph.AnimGraphNode_StateMachine_0.Boosting.AnimStateNode_1.BoostReady
  nodes [Output Animation Pose(Result=Sequence Player 'MeatBallVehicle_Engine_Boost_Ready_Vehicle'() {bLoopAnimation=false, sequence=/Game/Animation/Game/Vehicles/MeatBall/MeatBallVehicle_Engine_Boost_Ready_Vehicle})]
 }
 BoostLoop {
  path=AnimGraph.AnimGraphNode_StateMachine_0.Boosting.AnimStateNode_2.BoostLoop
  nodes [Output Animation Pose(Result=Sequence Player 'MeatBallVehicle_Engine_Boost_Loop_Vehicle'() {sequence=/Game/Animation/Game/Vehicles/MeatBall/MeatBallVehicle_Engine_Boost_Loop_Vehicle})]
 }
 BoostOnCD {
  path=AnimGraph.AnimGraphNode_StateMachine_0.Boosting.AnimStateNode_3.BoostOnCD
  nodes [Output Animation Pose(Result=Sequence Player 'MeatBallVehicle_Engine_Boost_OnCD_Vehicle'() {bLoopAnimation=false, sequence=/Game/Animation/Game/Vehicles/MeatBall/MeatBallVehicle_Engine_Boost_OnCD_Vehicle})]
 }
 Idle {
  path=AnimGraph.AnimGraphNode_StateMachine_0.Boosting.AnimStateNode_4.Idle
  nodes [Output Animation Pose(Result=Sequence Player 'MeatBallVehicle_BasePose_Vehicle'() {sequence=/Game/Animation/Game/Vehicles/MeatBall/MeatBallVehicle_BasePose_Vehicle})]
 }
 Transition {
  path=AnimGraph.AnimGraphNode_StateMachine_0.Boosting.AnimStateTransitionNode_0.Transition
  nodes [Result(bCanEnterTransition=Get bIsBoostReady() -> (bIsBoostReady:bool))]
 }
 Transition {
  path=AnimGraph.AnimGraphNode_StateMachine_0.Boosting.AnimStateTransitionNode_1.Transition
  nodes [Result(bCanEnterTransition=Get bIsBoosting() -> (bIsBoosting:bool))]
 }
 Transition {
  path=AnimGraph.AnimGraphNode_StateMachine_0.Boosting.AnimStateTransitionNode_2.Transition
  nodes [Result(bCanEnterTransition=Get bIsBoostEnding() -> (bIsBoostEnding:bool))]
 }
 Transition {
  path=AnimGraph.AnimGraphNode_StateMachine_0.Boosting.AnimStateTransitionNode_4.Transition
  nodes [Result(bCanEnterTransition=Get bIsBoostReady())]
 }
 Transition {
  path=AnimGraph.AnimGraphNode_StateMachine_0.Boosting.AnimStateTransitionNode_5.Transition
  nodes [Result(bCanEnterTransition=NOT Boolean(A=Get bHasDriver() -> (bHasDriver:bool)))]
 }
 Transition {
  path=AnimGraph.AnimGraphNode_StateMachine_0.Boosting.AnimStateTransitionNode_6.Transition
  nodes [Result(bCanEnterTransition=NOT Boolean(A=Get bHasDriver()))]
 }
 Transition {
  path=AnimGraph.AnimGraphNode_StateMachine_0.Boosting.AnimStateTransitionNode_7.Transition
  nodes [Result(bCanEnterTransition=NOT Boolean(A=Get bHasDriver()))]
 }
 Rocket {
  entry=Default
  path=AnimGraph.AnimGraphNode_StateMachine_1.Rocket
  ids=id
  nodes [
   AnimStateNode_0 = Default() {is=title AnimStateNode}
   AnimStateNode_1 = RocketOnCD() {is=title AnimStateNode}
   AnimStateNode_2 = RocketReady() {is=title AnimStateNode}
   AnimStateTransitionNode_0 = AnimStateTransitionNode() {crossfadeDuration=0}
   AnimStateTransitionNode_1 = AnimStateTransitionNode()
   AnimStateTransitionNode_2 = AnimStateTransitionNode() {bAutomaticRuleBasedOnSequencePlayerInState=true}
  ]
  links [
   {
    dst=RocketOnCD
    rule=AnimGraph.AnimGraphNode_StateMachine_1.Rocket.AnimStateTransitionNode_0.Transition
    src=Default
    via=AnimStateTransitionNode_0
   }
   {
    dst=RocketReady
    rule=AnimGraph.AnimGraphNode_StateMachine_1.Rocket.AnimStateTransitionNode_1.Transition
    src=RocketOnCD
    via=AnimStateTransitionNode_1
   }
   {
    dst=Default
    rule=AnimGraph.AnimGraphNode_StateMachine_1.Rocket.AnimStateTransitionNode_2.Transition
    src=RocketReady
    via=AnimStateTransitionNode_2
   }
  ]
 }
 Default {path=AnimGraph.AnimGraphNode_StateMachine_1.Rocket.AnimStateNode_0.Default, nodes [Output Animation Pose(Result=Additive Identity Pose())]}
 RocketOnCD {
  path=AnimGraph.AnimGraphNode_StateMachine_1.Rocket.AnimStateNode_1.RocketOnCD
  nodes [Output Animation Pose(Result="Sequence Player (additive) 'MeatBallVehicle_RocketOnCD_Vehicle'"() {bLoopAnimation=false, sequence=/Game/Animation/Game/Vehicles/MeatBall/MeatBallVehicle_RocketOnCD_Vehicle})]
 }
 RocketReady {
  path=AnimGraph.AnimGraphNode_StateMachine_1.Rocket.AnimStateNode_2.RocketReady
  nodes [Output Animation Pose(Result="Sequence Player (additive) 'MeatBallVehicle_RocketReady_Vehicle'"() {bLoopAnimation=false, sequence=/Game/Animation/Game/Vehicles/MeatBall/MeatBallVehicle_RocketReady_Vehicle})]
 }
 Transition {
  path=AnimGraph.AnimGraphNode_StateMachine_1.Rocket.AnimStateTransitionNode_0.Transition
  nodes [Result(bCanEnterTransition=Get bIsRocketOnCoolDown() -> (bIsRocketOnCoolDown:bool))]
 }
 Transition {
  path=AnimGraph.AnimGraphNode_StateMachine_1.Rocket.AnimStateTransitionNode_1.Transition
  nodes [Result(bCanEnterTransition=Get bIsRocketReady() -> (bIsRocketReady:bool))]
 }
 Transition {path=AnimGraph.AnimGraphNode_StateMachine_1.Rocket.AnimStateTransitionNode_2.Transition, nodes [Result()]}
 EventGraph {nodes [
   Event BlueprintUpdateAnimation()
   n1 = Clamp Angle!(
    AngleDegrees:double=BreakRotator(InRot=Get SteeringRotator() -> (SteeringRotator:Rotator)).Yaw
    MinAngleDegrees:double=-45
    MaxAngleDegrees:double=45
   ) -> (ReturnValue:double)
   > Set jet_rotation(jet_rotation:Rotator=RInterpTo!(Current:Rotator="0, 0, 0", Target:Rotator=MakeRotator(Yaw=@n1), DeltaTime:float=0.1, InterpSpeed:float=5) -> (ReturnValue:Rotator)) -> (Output_Get:Rotator)
  ]}
]
variables [jet_rotation {type=Rotator}]
