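; Audio controller for the Meatball motorboat, driving engine, wake, scrape and impact sounds.
; Update events map throttle, boost, speed, steering and water/air state onto FortVehicleAudioVoice parameters, while engine one-shots receive water and speed flags.
; CacheAudio hands the voices to the native vehicle off dedicated servers, dropping wake-turn audio on mobile.
; Collision sounds require minimum speed, pause briefly after a driver enters, and shift upward for locally viewed occupants.
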(event EventBeginPlay
  (Utilities|FlowControl|Delay 0.2)
  (Timeline|Initialize PlayFromStart
    (:Update
      (Class|FortVehicleAudioVoice|SetParam (Variables|Default|GetAudio_Engine_Down) 1.0))))

(event Custom|Audio%20Tick%20Update
  (params (Vehicle :type Meatball%20Vehicle%20L%20Object%20Reference))
  (Networking|CanExecuteCosmeticEvents
    (:True
      (Variables|Default|SetVehicle Vehicle))))

(event Custom|On%20Sleep%20State%20Change
  (params (Awake :type Boolean))
  (if Awake
    (bind _vehicle (Variables|Default|GetVehicle)
      (:then
        (Vehicle|BindEventtoOnPawnEnteredVehicleatSeat _vehicle (EventDispatchers|CreateEvent :!event Custom|On%20Enter%20Vehicle%20At%20Seat))
        (Vehicle|BindEventtoOnPawnExitedVehiclefromSeat _vehicle (EventDispatchers|CreateEvent :!event Custom|On%20Exit%20Vehicle%20At%20Seat)))
      (:else))))

(event Custom|Stop%20All%20Audio
  (Class|FortVehicleAudioVoice|StopAllAudio (Variables|Default|GetAudio_Wake_Turn)))

(event Custom|On%20Enter%20Vehicle%20At%20Seat
  (params (Vehicle :type Fort%20Vehicle%20Interface%20Interface%20%28by%20ref%29) (PlayerPawn :type Fort%20Pawn%20Object%20Reference) (SeatIndex :type Integer))
  (CallFunction|LocalSeat PlayerPawn)
  (if (== SeatIndex 0)
    (Variables|Default|SetInitialAwaken true)
    (bind _returnvalue (Utilities|Time|SetTimerbyEvent (delegate Custom|UpdateDriverEntering) 1.0))
    (Variables|Default|SetDriver%20Entering%20Handle _returnvalue)))

(event Custom|UpdateDriverEntering
  (Variables|Default|SetInitialAwaken false))

(event Custom|On%20Exit%20Vehicle%20At%20Seat
  (params (Vehicle :type Fort%20Vehicle%20Interface%20Interface%20%28by%20ref%29) (PlayerPawn :type Fort%20Pawn%20Object%20Reference) (SeatIndex :type Integer))
  (bind _driver_entering_handle (Variables|Default|GetDriver%20Entering%20Handle))
  (CallFunction|LocalSeat 0)
  (if (== SeatIndex 0)
    (Variables|Default|SetInitialAwaken false)
    (if (Utilities|Time|IsValidTimerHandle _driver_entering_handle)
      (Utilities|Time|ClearandInvalidateTimerbyHandle _driver_entering_handle))))

(event Custom|Update%20Engine
  (bind _vehicle (Variables|Default|GetVehicle))
  (bind _returnvalue (Class|FortVehicleInterface|IsBoosting _vehicle))
  (bind _returnvalue_2 (Vehicle|GetForwardSpeedKmh _vehicle))
  (bind _returnvalue_3 (Math|Float|MapRangeClamped _returnvalue_2 0.0 1e+02 0.2 1.0))
  (bind _returnvalue_4 (Vehicle|PlayerIsTryingtoMoveForward _vehicle))
  (bind _returnvalue_5 (not _returnvalue))
  (bind _returnvalue_6 (and _returnvalue_4 _returnvalue_5))
  (bind _returnvalue_7 (Math|Conversions|ToFloat(Boolean) _returnvalue_6))
  (bind _returnvalue_8 (Class|FortVehicleInterface|IsBoosting _vehicle))
  (bind _returnvalue_9 (Vehicle|PlayerIsTryingtoMoveForward _vehicle))
  (bind _returnvalue_10 (or _returnvalue_9 _returnvalue_8))
  (bind _returnvalue_11 (not _returnvalue_10))
  (bind _returnvalue_12 (Math|Conversions|ToFloat(Boolean) _returnvalue_11))
  (Utilities|FlowControl|Sequence
    (:then_0
      (Class|FortVehicleAudioVoice|SetParam (Variables|Default|GetAudio_Engine_Up) _returnvalue_7)
      (Class|FortVehicleAudioVoice|SetParamExplicit (Variables|Default|GetAudio_Engine_Up) "ThrottleOn" _returnvalue_7)
      (Class|FortVehicleAudioVoice|SetParamExplicit (Variables|Default|GetAudio_Engine_Up) "Speed" _returnvalue_3)
      (Class|FortVehicleAudioVoice|SetParamExplicit (Variables|Default|GetAudio_Engine_Up) "Steer" (Math|Float|MapRangeClamped (Math|Float|Absolute(Float) (Vehicle|GetSteeringAngle _vehicle)) 0.0 2e+01 0.0 1.0))
      (Class|FortVehicleAudioVoice|SetParamExplicit (Variables|Default|GetAudio_Engine_Up) "IsInWater" (Math|Conversions|ToFloat(Boolean) (Class|FortVehicleInterface|IsinWater (Variables|Default|GetVehicle))))
      (Class|FortVehicleAudioVoice|SetParamExplicit (Variables|Default|GetAudio_Engine_Up) "IsBoosting" (Math|Conversions|ToFloat(Boolean) (Class|FortVehicleInterface|IsBoosting (Variables|Default|GetVehicle)))))
    (:then_1
      (Class|FortVehicleAudioVoice|SetParam (Variables|Default|GetAudio_Engine_Down) _returnvalue_12)
      (Class|FortVehicleAudioVoice|SetParamExplicit (Variables|Default|GetAudio_Engine_Down) "ThrottleOff" _returnvalue_12)
      (Class|FortVehicleAudioVoice|SetParamExplicit (Variables|Default|GetAudio_Engine_Down) "Speed" _returnvalue_3)
      (Class|FortVehicleAudioVoice|SetParamExplicit (Variables|Default|GetAudio_Engine_Down) "HasDriver" (Math|Conversions|ToFloat(Boolean) (Vehicle|HasDriver (Variables|Default|GetVehicle)))))))

(event Custom|Update%20Wake
  (bind _vehicle (Variables|Default|GetVehicle))
  (bind _returnvalue (Vehicle|GetSteeringAngle _vehicle))
  (bind _returnvalue_2 (Math|Float|Absolute(Float) _returnvalue))
  (bind _vehicle_2 (Variables|Default|GetVehicle))
  (bind _returnvalue_3 (Vehicle|GetForwardSpeedKmh _vehicle_2))
  (bind _returnvalue_4 (Math|Float|Absolute(Float) _returnvalue_3))
  (bind _returnvalue_5 (Math|Float|MapRangeClamped _returnvalue_4 0.0 5e+01 0.0 1.0))
  (bind _returnvalue_6 (Math|Float|MapRangeClamped _returnvalue_2 0.0 2e+01 0.0 1.0))
  (bind _vehicle_3 (Variables|Default|GetVehicle))
  (bind _returnvalue_7 (Vehicle|GetForwardSpeedKmh _vehicle_3))
  (bind _returnvalue_8 (Math|Float|Absolute(Float) _returnvalue_7))
  (bind _returnvalue_9 (Math|Float|MapRangeClamped _returnvalue_8 0.0 5e+01 0.0 1.0))
  (bind _vehicle_4 (Variables|Default|GetVehicle))
  (bind _returnvalue_10 (Class|FortVehicleInterface|IsinWater _vehicle_4))
  (bind _returnvalue_11 (Math|Conversions|ToFloat(Boolean) _returnvalue_10))
  (bind _returnvalue_12 (* _returnvalue_6 _returnvalue_5))
  (bind _audio_wake_speed (Variables|Default|GetAudio_Wake_Speed))
  (bind _audio_wake_turn (Variables|Default|GetAudio_Wake_Turn))
  (Class|FortVehicleAudioVoice|SetParam _audio_wake_turn _returnvalue_12)
  (Class|FortVehicleAudioVoice|SetParamExplicit _audio_wake_turn "Steer" _returnvalue_12)
  (Class|FortVehicleAudioVoice|SetParamExplicit (Variables|Default|GetAudio_Wake_Turn) "IsInWater" _returnvalue_11)
  (Class|FortVehicleAudioVoice|SetParam _audio_wake_speed _returnvalue_9)
  (Class|FortVehicleAudioVoice|SetParamExplicit _audio_wake_speed "Speed" _returnvalue_9)
  (Class|FortVehicleAudioVoice|SetParamExplicit _audio_wake_speed "IsInWater" _returnvalue_11)
  (Class|FortVehicleAudioVoice|SetParamExplicit _audio_wake_speed "IsInAir" (Math|Conversions|ToFloat(Boolean) (Vehicle|VehicleIsinAir _vehicle_4))))

(event Custom|Update%20Oneshots
  (bind _vehicle (Variables|Default|GetVehicle))
  (bind _returnvalue (Class|FortVehicleInterface|IsinWater _vehicle))
  (bind _returnvalue_2 (Math|Conversions|ToInteger(Boolean) _returnvalue))
  (Utilities|FlowControl|Sequence
    (:then_0
      (bind _currentengineextdown (Variables|Default|GetCurrentEngineExtDown)
        (:then
          (Audio|Parameter|SetIntegerParameter _currentengineextdown "IsInWater" _returnvalue_2 :!class /Script/AudioExtensions.AudioParameterControllerInterface))
        (:else)))
    (:then_1
      (bind _currentengineextup (Variables|Default|GetCurrentEngineExtUp)
        (:then
          (Audio|Parameter|SetIntegerParameter _currentengineextup "IsInWater" _returnvalue_2 :!class /Script/AudioExtensions.AudioParameterControllerInterface))
        (:else)))))

(event OnOneshotEvent(Audio_Engine_Up)
  (params (AudioComponent :type Audio%20Component%20Object%20Reference) (Index :type Integer))
  (bind _vehicle (Variables|Default|GetVehicle))
  (Variables|Default|GetVehicle
    (:then
      (Audio|Parameter|SetIntegerParameter AudioComponent "IsInWater" (Math|Conversions|ToInteger(Boolean) (Class|FortVehicleInterface|IsinWater _vehicle)) :!class /Script/AudioExtensions.AudioParameterControllerInterface)
      (Audio|Parameter|SetIntegerParameter AudioComponent "SpeedThreshold" (Math|Conversions|ToInteger(Boolean) (> (Vehicle|GetTotalSpeedKmh _vehicle) 1e+01)) :!class /Script/AudioExtensions.AudioParameterControllerInterface))
    (:else)))

(event Custom|Update%20Scrape
  (bind _audio_scrape (Variables|Default|GetAudio_Scrape))
  (bind _vehicle (Variables|Default|GetVehicle))
  (Class|FortVehicleAudioVoice|SetParam _audio_scrape (Math|Conversions|ToFloat(Boolean) (and (not (Class|FortVehicleInterface|IsinWater _vehicle)) (not (Vehicle|VehicleIsinAir _vehicle)))))
  (Class|FortVehicleAudioVoice|SetParamExplicit _audio_scrape "Speed" (Math|Float|MapRangeClamped (Math|Float|Absolute(Float) (Vehicle|GetForwardSpeedKmh _vehicle)) 0.0 3e+01 0.0 1.0)))

(event OnOneshotEvent(Audio_Engine_Down)
  (params (AudioComponent :type Audio%20Component%20Object%20Reference) (Index :type Integer))
  (bind _vehicle (Variables|Default|GetVehicle))
  (Variables|Default|GetVehicle
    (:then
      (Audio|Parameter|SetIntegerParameter AudioComponent "IsInWater" (Math|Conversions|ToInteger(Boolean) (Class|FortVehicleInterface|IsinWater _vehicle)) :!class /Script/AudioExtensions.AudioParameterControllerInterface)
      (Audio|Parameter|SetIntegerParameter AudioComponent "SpeedThreshold" (Math|Conversions|ToInteger(Boolean) (> (Vehicle|GetTotalSpeedKmh _vehicle) 1e+01)) :!class /Script/AudioExtensions.AudioParameterControllerInterface))
    (:else)))

(fn CacheAudio
  (params (Meatball :type Fort%20Meatball%20Vehicle%20Object%20Reference))
  (bind _returnvalue (Fortnite|IsMobilePlatform))
  (if (not (Networking|IsDedicatedServer))
    (Class|FortMeatballVehicle|CacheAudioPointers Meatball (Variables|Default|GetAudio_Engine_Up) (Variables|Default|GetAudio_Engine_Down) (select _returnvalue 0 (Variables|Default|GetAudio_Wake_Turn)) (Variables|Default|GetAudio_Wake_Speed) (Variables|Default|GetAudio_Scrape) (Variables|Default|GetAudio_Engine_Out_Of_Fuel))
    (Variables|Default|SetVehicle (Utilities|Casting|CastToMeatballVehicle_L (Actor|GetAttachParentActor)))
    (if _returnvalue
      (Components|DestroyComponent (Variables|Default|GetAudio_Wake_Turn)))))

(fn LocalSeat
  (params (LocalPawnInSeat :type Fort%20Pawn%20Object%20Reference))
  (Utilities|IsValid LocalPawnInSeat
    (:Is%20Valid
      (if (Pawn|IsLocallyViewed LocalPawnInSeat)
        (Variables|Default|SetLocalPawnInSeat LocalPawnInSeat)))
    (:Is%20Not%20Valid
      (Variables|Default|SetLocalPawnInSeat 0))))

(fn OnCollision
  (params (HitLocation :type Vector%20%28by%20ref%29) (HitNormalImpulse :type Vector%20%28by%20ref%29))
  (if (Variables|Default|GetInitialAwaken)
    _
    (elif (> (Vehicle|GetTotalSpeedKmh (Variables|Default|GetVehicle)) 3.0)
      (bind (_outresult _outmagnitude) (Audio|OnCollision (Variables|Default|GetFortBoatCollisionImpacts) (select (Utilities|IsValid (Variables|Default|GetLocalPawnInSeat) :!class /Script/Engine.KismetSystemLibrary) (+ HitLocation "0,0,175.000000") HitLocation) HitNormalImpulse))
      (return _outresult _outmagnitude))))

(fn OnPontoonEnterWater
  (params (Pontoon :type Spherical%20Pontoon%20Structure%20%28by%20ref%29))
  (bind (_centersocket _relativelocation _radius _bfxenabled _localforce _centerlocation) (Utilities|Struct|BreakSphericalPontoon Pontoon))
  (Audio|OnCollision (Variables|Default|GetFortBoatPontoonImpacts) _centerlocation _localforce)
  (return))

(fn ConstructionScript)
