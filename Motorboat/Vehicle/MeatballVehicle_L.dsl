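; Event graph for the Fortnite Meatball motorboat, handling driver and passenger entry and exit, boost, damage, collisions and water interaction.
; Pontoon, boost and health events drive Niagara/Cascade effects, camera shake, rumble and body material, while a child-actor controller handles much of the audio.
; Server-side logic covers speed-scaled land damage timers and quest stats, including a rare legendary-boat roll.
; Hits can also drop loot, apply pickaxe impulses or launch struck characters.
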(event EventOnDisabledChanged
  (params (bDisabled :type Boolean))
  ;[1 1 1] EVENT -> On Disabled Changed
  (|CallParentFunction bDisabled)
  (Development|PrintString "on disabled Changed")
  (if bDisabled
    (Effects|Components|ParticleSystem|SetBoolParameter (Variables|Default|GetFX_Boost) "isReady" false)
    (elif (Vehicle|HasDriver)
      (CallFunction|CheckForBoostReady))))

(event Custom|StopStartupRumble
  ;[1 1 1] Engine Start Force Feedback
  (Utilities|FlowControl|Sequence
    (:then_0
      (Game|Feedback|PlayDynamicForceFeedback (Variables|Meatball|GetBP%20Driving%20Player%20Controller) (stale _newtrack_0) -1 true true true true))
    (:then_1
      (bind (_direction _newtrack_0) (Timeline|Timeline_0 Stop)
        (:Update
          _)
        (:Finished
          _)))))

(event Custom|PlayStartupFF
  ;[1 1 1] Engine Start Force Feedback
  (Networking|CanExecuteCosmeticEvents
    (:True
      (if (Pawn|IsLocallyControlled (Variables|Meatball|GetDriverPawn))
        (Utilities|FlowControl|Sequence
          (:then_0
            (Game|Feedback|PlayDynamicForceFeedback (Variables|Meatball|GetBP%20Driving%20Player%20Controller) (stale Custom|StopStartupRumble _newtrack_0) -1 true true true true))
          (:then_1
            (Timeline|Timeline_0 PlayFromStart)))))))

(event Custom|FireFF
  (params (GameplayAbility :type Fort%20Gameplay%20Ability%20Object%20Reference) (AbilityTags :type Gameplay%20Tag%20Container%20Structure))
  ;[1 1 1] Missile Fired Force Feedback
  (Game|Feedback|ClientPlayForceFeedback (Fortnite|GetFortPlayerControllerfromActor (Variables|Meatball|GetDriverPawn)) (Variables|Default|GetForceFeedbackWeaponFire) "None" false false false))

(event EventOnSleepStateChanged
  (params (bAwake :type Boolean))
  (bind _audiocontroller (Variables|Default|GetAudioController))
  ;[1 1 1] On Sleep State Change
  (|CallParentFunction bAwake)
  (Utilities|IsValid _audiocontroller
    (:Is%20Valid
      (bind _asbp_meatball_audio_controller (Utilities|Casting|CastToBP_Meatball_AudioController (Variables|/Script/Engine.ChildActorComponent|GetChildActor _audiocontroller)))
      (Class|BPMeatballAudioController|OnSleepStateChange _asbp_meatball_audio_controller bAwake))))

(event Custom|Audio%20Tick%20Update
  ;[1 1 1] Audio Tick Update
  (bind _asbp_meatball_audio_controller (Utilities|Casting|CastToBP_Meatball_AudioController (Variables|/Script/Engine.ChildActorComponent|GetChildActor (Variables|Default|GetAudioController))))
  (Class|BPMeatballAudioController|AudioTickUpdate _asbp_meatball_audio_controller self))

(event EventOnStopTick
  ;[1 1 1] EVENT -> Stop Tick
  (bind _asbp_meatball_audio_controller (Utilities|Casting|CastToBP_Meatball_AudioController (Variables|/Script/Engine.ChildActorComponent|GetChildActor (Variables|Default|GetAudioController))))
  (Class|BPMeatballAudioController|StopAllAudio _asbp_meatball_audio_controller))

(event Custom|LandDamageEvent
  (bind _returnvalue (Vehicle|GetForwardSpeedKmh))
  ;[1 0.9429 0.474] Damage Vehicle if moving on Land
  (Utilities|IsValid (Variables|Meatball|GetDriverPawn)
    (:Is%20Valid
      (bind _returnvalue_2 (Utilities|Time|SetTimerbyEvent (delegate Custom|LandDamage) (Variables|Meatball|GetLandDamageFrequency) true))
      (Variables|Meatball|SetLandDamageTimer _returnvalue_2)
      (Networking|SwitchHasAuthority
        (:Authority
          (if (and (CallFunction|IsonGround) (or (> _returnvalue 2.0) (< _returnvalue -2.0)))
            (Variables|Meatball|SetLandDamageFrequency (- 1.0 (Math|Float|Clamp(Float) (/ (Math|Float|Absolute(Float) _returnvalue) 7e+01) 0.0 0.8)))
            (Environment|ApplyGlobalEnvironmentDamagetoActor self (Variables|Meatball|GetBoatDamageOnLand) "(GameplayTags=((TagName=\"Analytics.Vehicle.Meatball.Damage.DriveOnLand\")))")
            (else
              (CallFunction|StopLandRumble)
              (Variables|Meatball|SetLandDamageFrequency 0.25))))))
    (:Is%20Not%20Valid
      (Utilities|Time|ClearandInvalidateTimerbyHandle (Variables|Meatball|GetLandDamageTimer)))))

(event Custom|LandDamage
  (bind _returnvalue (Vehicle|GetForwardSpeedKmh))
  ;[1 0.9429 0.474] Damage Vehicle if moving on Land
  (Utilities|IsValid (Variables|Meatball|GetDriverPawn)
    (:Is%20Valid
      (bind _returnvalue_2 (Utilities|Time|SetTimerbyEvent (delegate Custom|LandDamage) (Variables|Meatball|GetLandDamageFrequency) true))
      (Variables|Meatball|SetLandDamageTimer _returnvalue_2)
      (Networking|SwitchHasAuthority
        (:Authority
          (if (and (CallFunction|IsonGround) (or (> _returnvalue 2.0) (< _returnvalue -2.0)))
            (Variables|Meatball|SetLandDamageFrequency (- 1.0 (Math|Float|Clamp(Float) (/ (Math|Float|Absolute(Float) _returnvalue) 7e+01) 0.0 0.8)))
            (Environment|ApplyGlobalEnvironmentDamagetoActor self (Variables|Meatball|GetBoatDamageOnLand) "(GameplayTags=((TagName=\"Analytics.Vehicle.Meatball.Damage.DriveOnLand\")))")
            (else
              (CallFunction|StopLandRumble)
              (Variables|Meatball|SetLandDamageFrequency 0.25))))))
    (:Is%20Not%20Valid
      (Utilities|Time|ClearandInvalidateTimerbyHandle (Variables|Meatball|GetLandDamageTimer)))))

(event EventOnHoldExitStarted
  (params (ExitingPawn :type Fort%20Pawn%20Object%20Reference) (ExitDuration :type Float%20%28single-precision%29))
  ;[1 1 1] EVENT -> OnHoldExitStarted/Stopped
  (Ability|SendGameplayEventtoActor ExitingPawn "(TagName=\"Athena.Vehicle.Biplane.OnHoldExitStarted\")" (Utilities|Struct|MakeGameplayEventData :ContextHandle "()" :InstigatorTags "(GameplayTags=)" :TargetTags "(GameplayTags=)" :EventMagnitude ExitDuration :TargetData "()")))

(event EventBeginPlay
  (bind _skeletalmesh (Variables|Vehicle|GetSkeletalMesh))
  ;[1 1 1] EVENT -> Begin Play
  (if (not (ScalableFloat|GetValueasBool (Variables|Meatball|GetbEnableWeapons)))
    (Components|DestroyComponent (Variables|Default|GetFortVehicleSeatWeapon)))
  (Health|BindEventtoOnHealthChanged (EventDispatchers|CreateEvent :!event Custom|EventOnHealthChanged))
  (Health|BindEventtoOnHandleOutOfHealthDelegate (EventDispatchers|CreateEvent :!event Custom|EventOnHandleOutOfHealth))
  (Utilities|FlowControl|Sequence
    (:then_0
      ;; If Server Kill Audio/FX
      (if (Networking|IsDedicatedServer)
        (Components|DestroyComponent (Variables|FX|GetWaterEffectsComponent))))
    (:then_1
      ;;[1 0.8611 0.5571] Boaty McBoatface achievement - boat has 0.1% chance of rolling Legendary
      (Networking|SwitchHasAuthority
        (:Authority
          (bind _output_get (Variables|Quests|SetIsLegendary (== (Math|Random|RandomInteger 1000) 0)))
          (if _output_get
            (Development|PrintString "I'm a legenary boat")))))
    (:then_2
      (bind _returnvalue (Rendering|Material|CreateDynamicMaterialInstance _skeletalmesh 2 (Rendering|Material|GetMaterial (Variables|Vehicle|GetSkeletalMesh) 2) "None" :!class /Script/Engine.PrimitiveComponent))
      (bind _output_get_2 (Variables|VFX|SetBoostGuageMIC _returnvalue))
      (Rendering|Material|SetMaterial _skeletalmesh 2 _output_get_2)
      (Rendering|Material|SetScalarParameterValue _output_get_2 "Enable" 0.0 :!class /Script/Engine.MaterialInstanceDynamic)
      (bind _asbp_meatball_audio_controller (Utilities|Casting|CastToBP_Meatball_AudioController (Variables|/Script/Engine.ChildActorComponent|GetChildActor (Variables|Default|GetAudioController))))
      (Class|BPMeatballAudioController|CacheAudio _asbp_meatball_audio_controller self))
    (:then_3
      (bind _returnvalue_2 (Actor|GetActorOfClass "/Game/Athena/Apollo/Environments/Blueprints/CalendarEvents/BP_ApolloSnowSetup.BP_ApolloSnowSetup_C"))
      (if (Utilities|IsValid _returnvalue_2 :!class /Script/Engine.KismetSystemLibrary)
        (bind _output_get_3 (Variables|Default|SetSet%20up%20snow%20reference _returnvalue_2))
        (Variables|Default|SetSet%20up%20Snow%20Amount (Variables|/Game/Athena/Apollo/Environments/Blueprints/CalendarEvents/BP_ApolloSnowSetup.BP_ApolloSnowSetup_C|GetSnowAmount _output_get_3))))
    (:then_4
      (CallFunction|CacheGroundFXComponent (Variables|Default|GetFX_GroundInteraction)))))

(event EventOnHoldExitStopped
  (params (ExitingPawn :type Fort%20Pawn%20Object%20Reference))
  ;[1 1 1] EVENT -> OnHoldExitStarted/Stopped
  (Ability|SendGameplayEventtoActor ExitingPawn "(TagName=\"Athena.Vehicle.Biplane.OnHoldExitStopped\")"))

(event EventOnDamagePlayEffects
  (params (Damage :type Float%20%28single-precision%29) (DamageTags :type Gameplay%20Tag%20Container%20Structure%20%28by%20ref%29) (Momentum :type Vector :default "0, 0, 0") (HitInfo :type Hit%20Result%20Structure%20%28by%20ref%29) (InstigatedBy :type Fort%20Pawn%20Object%20Reference) (DamageCauser :type Actor%20Object%20Reference) (EffectContext :type Gameplay%20Effect%20Context%20Handle%20Structure))
  (bind (_bblockinghit _binitialoverlap _time _distance _location _impactpoint _normal) (Collision|BreakHitResult HitInfo))
  ;[1 1 1] Spawn Large Sparks when Damage is taken
  (if (> Damage 21.0)
    (Niagara|SpawnSystematLocation (Variables|Default|GetFX%20Large%20Impact) _location (Math|Conversions|RotationFromXVector _normal) (Math|Conversions|ToVector(Float) (Math|Float|SelectFloat 1.5 1.0 (> Damage 5e+01))) true true "None" false)))

(event EventK2_ApplyCosmeticWrap
  (params (LoadedWrap :type Athena%20Item%20Wrap%20Definition%20Object%20Reference))
  (bind _output_materialinterface (Utilities|Struct|BreakSkeletalMaterial (Utilities|Array|Get(acopy) (Variables|/Script/Engine.SkeletalMesh|GetMaterials (Variables|/Script/Engine.SkinnedMeshComponent|GetSkeletalMesh (Variables|Vehicle|GetSkeletalMesh))))))
  ;[1 1 1] EVENT -> Apply Cosmetic Wrap
  (if (not (GameplayTags|HasMatchingGameplayTag "(TagName=\"Gameplay.Status.Chrome\")"))
    (Class|FortVehicleInterface|ApplyMaterialandUseWrapifPossible "/Game/Athena/DrivableVehicles/Meatball/Material/Meatball_Parts.Meatball_Parts" 1)
    (Class|FortVehicleInterface|ApplyMaterialandUseWrapifPossible _output_materialinterface 0 (delegate Custom|OnMIDCreated_SlotZero))))

(event Custom|OnMIDCreated_SlotZero
  (params (MID :type Material%20Instance%20Dynamic%20Object%20Reference))
  ;[1 1 1] EVENT -> Apply Cosmetic Wrap
  (bind _output_get (Variables|Meatball|SetBodyMID MID))
  (Rendering|Material|SetScalarParameterValue _output_get "Damage" (Variables|VFX|GetDamageValue) :!class /Script/Engine.MaterialInstanceDynamic))

(event EventOnPawnEnterVehicle
  (params (PlayerPawn :type Fort%20Player%20Pawn%20Object%20Reference) (SeatIdx :type Integer))
  ;[0 0.4 0] EVENT -> Pawn Enter Vehicle
  (|CallParentFunction PlayerPawn SeatIdx)
  (Utilities|FlowControl|Sequence
    (:then_0
      (Utilities|Array|AddUnique (Variables|Meatball|GetPassengerPawns) PlayerPawn)
      ;; SetLocalPlayerPawn
      (Utilities|FlowControl|Sequence
        (:then_0
          (if (Pawn|IsLocallyControlled PlayerPawn)
            (Variables|Meatball|SetLocalPlayerPawn PlayerPawn)))
        (:then_1
          _))
      (switch int SeatIdx
        (:0
          (CallFunction|ManageStateDriverEnter PlayerPawn))
        (:1
          _)
        (:2
          _)
        (:3
          _))
      (CallFunction|ManageStatePassengerEnter PlayerPawn))
    (:then_1
      ;;[1 0.8627 0.5569] Boaty McBoatface achievement - check quest status and send complex custom if this is a legendary boat
      (Networking|SwitchHasAuthority
        (:Authority
          (if (Variables|Quests|GetIsLegendary)
            (Quest|SendComplexCustomStatEvent (Quests|GetQuestManager (Fortnite|GetFortPlayerControllerfromActor PlayerPawn) "Invalid") 0 "(GameplayTags=)" "(GameplayTags=((TagName=\"Athena.Achievements.LegendaryBoat\")))" 1)))
        (:Remote
          (CallFunction|RefreshDamageState))))
    (:then_2
      (bind _returnvalue (Environment|ApplyGlobalEnvironmentGameplayEffecttoActor PlayerPawn (Variables|Default|GetGE_Boost))) ; Apply GE for mobile Boost radial
      (Variables|Default|SetCooldownCueEvent _returnvalue)
      ;[0 0.6 0] Add AthenaInValet GameplayTag
      (Networking|SwitchHasAuthority
        (:Authority
          (GameplayEffects|ApplyGameplayEffectToTarget :self (Ability|GetAbilitySystemComponent PlayerPawn) :GameplayEffectClass (Variables|Default|GetGE_Apply%20in%20Valet%20Tag) :Target (Ability|GetAbilitySystemComponent PlayerPawn) :Level 1.0))))))

(event OnEnteredWaterDelegate(FortMeatballPontoons)
  (params (Pontoon :type Spherical%20Pontoon%20Structure%20%28by%20ref%29))
  (bind (_centersocket _relativelocation _radius _bfxenabled _localforce _centerlocation _socketrotation _offset _waterheight _waterdepth _immersiondepth _waterplanelocation _waterplanenormal _watersurfaceposition _watervelocity _waterbodyindex _bisinwater _currentwaterbodycomponent) (Utilities|Struct|BreakSphericalPontoon Pontoon))
  (bind _watereffectscomponent (Variables|FX|GetWaterEffectsComponent))
  (bind _bp_driving_player_controller (Variables|Meatball|GetBP%20Driving%20Player%20Controller))
  ;[1 1 1] EVENT -> Pontoon Enter Water
  (Utilities|IsValid _currentwaterbodycomponent
    (:Is%20Valid
      (Utilities|FlowControl|Sequence
        (:then_0
          (if (Utilities|Name|Equal(Name) "MidPontoonBig" _centersocket)
            (Utilities|FlowControl|Sequence
              (:then_0
                (if (CallFunction|IsNiagaraEffectsEnabled)
                  (Water|SetWaterBodyComponent _watereffectscomponent "User.WaterBody" _currentwaterbodycomponent)
                  (Effects|Components|ParticleSystem|SetFloatParameter _watereffectscomponent "User.WaterHeight" _waterheight)
                  (Effects|Components|ParticleSystem|SetFloatParameter _watereffectscomponent "User.IsInWaterCenter" 1.0)))
              (:then_1
                (Development|PrintString (Utilities|String|ToString(Float) _immersiondepth))
                (if (>= _immersiondepth (Variables|Haptics|GetMinImpactToShake))
                  (Utilities|IsValid _bp_driving_player_controller
                    (:Is%20Valid
                      (CameraShakes|StartCameraShake :self (Variables|/Script/Engine.PlayerController|GetPlayerCameraManager _bp_driving_player_controller) :ShakeClass (Variables|CameraShake|GetWaterImpactCameraShake) :Scale (/ _immersiondepth 1e+01) :PlaySpace "CameraLocal")
                      (Game|Feedback|PlayDynamicForceFeedback (Variables|Meatball|GetBP%20Driving%20Player%20Controller) (Variables|Meatball|GetRumbleIntensity) 0.3 false true false true))))))
            (elif (CallFunction|IsNiagaraEffectsEnabled)
              (Effects|Components|ParticleSystem|SetFloatParameter (Variables|FX|GetWaterEffectsComponent) "User.IsInWaterCenter" 0.0))))
        (:then_1
          (switch name _centersocket
            (:RearPontoonBig
              (CallFunction|EnteredWaterEngine))
            (:MidPontoonBig
              (CallFunction|EnteredWaterMid))))
        (:then_2
          (bind _asbp_meatball_audio_controller (Utilities|Casting|CastToBP_Meatball_AudioController (Variables|/Script/Engine.ChildActorComponent|GetChildActor (Variables|Default|GetAudioController))))
          (Class|BPMeatballAudioController|OnPontoonEnterWater _asbp_meatball_audio_controller Pontoon))))))

(event EventOnPawnExitVehicle
  (params (PlayerPawn :type Fort%20Player%20Pawn%20Object%20Reference) (ExitSocketName :type Name :default "None"))
  ;[0.8 0.8 0] EVENT -> Pawn Exit Vehicle
  (|CallParentFunction PlayerPawn ExitSocketName)
  (Utilities|FlowControl|Sequence
    (:then_0
      (Utilities|IsValid PlayerPawn
        (:Is%20Valid
          (Utilities|Array|RemoveItem (Variables|Meatball|GetPassengerPawns) PlayerPawn)
          ;; IsLocalLeaving
          (Utilities|FlowControl|Sequence
            (:then_0
              (if (Utilities|Equal(Object) (Variables|Meatball|GetLocalPlayerPawn) PlayerPawn)
                (Variables|Meatball|SetLocalPlayerPawn 0)))
            (:then_1
              (Projectile|RemoveGameplayCue%28Looping%29Local PlayerPawn (Variables|GameplayCues|GetGC_VehicleScreenDrips) (Variables|GameplayCues|GetGC_ParamsEmpty)) ; Ensure Water Drops are removed if player exits
            ))
          (if (Utilities|Equal(Object) (Variables|Meatball|GetDriverPawn) PlayerPawn)
            (CallFunction|ManageStateDriverExit PlayerPawn)
            (Development|PrintString "Driver exit")
            (else
              (CallFunction|ManageStatePassengerExit PlayerPawn))))))
    (:then_1
      (Networking|SwitchHasAuthority
        (:Remote
          (CallFunction|UpdateDamageState (Variables|VFX|GetDamageValue)))))
    (:then_2
      ;[1 1 0] Remove AthenaInValet Gameplay Tag
      (Networking|SwitchHasAuthority
        (:Authority
          (GameplayEffects|RemoveActiveEffectswithGrantedTags (Ability|GetFortAbilitySystemComponent PlayerPawn) (Variables|Default|GetInValetVehicleTag)))))))

(event Custom|ReverseBoostTimeline
  ;[1 1 1 1 22] BOOST Events
  (bind (_direction _fov _rumble) (Timeline|BoostTimeline Reverse)
    (:Update
      (Variables|Meatball|SetBoost _fov)
      (Variables|Camera|SetCameraFOVOffset (Math|Float|Lerp 0.0 (Variables|Haptics|GetMaxBoostFOV) _fov))
      (Variables|Haptics|SetBoostRumbleIntensity _rumble)
      (CallFunction|UpdateBoostRumble))
    (:Finished
      (switch Utilities|FlowControl|Switch|SwitchonETimelineDirection _direction
        (:Forward
          (CallFunction|BoostCameraReverse))
        (:Backward
          (Variables|Haptics|SetBoostCameraActive false)
          (CallFunction|BoostRumbleEnd)
          (Variables|Meatball|SetisBoosting false))))))

(event Custom|StartBoostRumble
  ;[1 1 1 1 22] BOOST Events
  ;;[1 1 1] Boost Rumble
  (Game|Feedback|PlayDynamicForceFeedback (Variables|Meatball|GetBP%20Driving%20Player%20Controller) (Variables|Haptics|GetBoostRumbleIntensity) -1 true false true false))

(event Custom|UpdateBoostRumble
  ;[1 1 1 1 22] BOOST Events
  ;;[1 1 1] Boost Rumble
  (Game|Feedback|PlayDynamicForceFeedback (Variables|Meatball|GetBP%20Driving%20Player%20Controller) (Variables|Haptics|GetBoostRumbleIntensity) -1 true false true false))

(event EventOnCollisionHitEffects
  (params (HitLocation :type Vector%20%28by%20ref%29 :default "0, 0, 0") (HitNormalImpulse :type Vector%20%28by%20ref%29 :default "0, 0, 0") (HitFrictionImpulse :type Vector%20%28by%20ref%29 :default "0, 0, 0") (HitNormal :type Vector%20%28by%20ref%29 :default "0, 0, 0") (HitActor :type Actor%20Object%20Reference) (HitSurfaceType :type EPhysicalSurface%20Enum :default "SurfaceType_Default"))
  ;[1 1 1] EVENT -> Collision Audio/FX
  (bind _asbp_meatball_audio_controller (Utilities|Casting|CastToBP_Meatball_AudioController (Variables|/Script/Engine.ChildActorComponent|GetChildActor (Variables|Default|GetAudioController))))
  (bind (_outresult _outmagnitude) (Class|BPMeatballAudioController|OnCollision _asbp_meatball_audio_controller HitLocation HitNormalImpulse))
  (if _outresult
    (CallFunction|PlayHapticsonHit _outmagnitude)))

(event Custom|BoostRumbleEnd
  ;[1 1 1 1 22] BOOST Events
  ;;[1 1 1] Boost Rumble
  (Game|Feedback|PlayDynamicForceFeedback (Variables|Meatball|GetBP%20Driving%20Player%20Controller) (Variables|Haptics|GetBoostRumbleIntensity) -1 true false true false))

(event Custom|BoostCameraReverse
  ;[1 1 1 1 22] BOOST Events
  (if (Variables|Haptics|GetBoostCameraActive)
    (CallFunction|ReverseBoostTimeline)))

(event EventOnBoostStarted
  (bind _fx_boost (Variables|Default|GetFX_Boost))
  ;[1 1 1 1 22] BOOST Events
  (Utilities|FlowControl|Sequence
    (:then_0
      (CallFunction|StartBoostRumble)
      (Variables|Haptics|SetBoostCameraActive true)
      (Variables|Meatball|SetisBoosting true)
      (Timeline|BoostTimeline Play))
    (:then_1
      (Effects|Components|ParticleSystem|SetBoolParameter _fx_boost "isActive" true)
      (Effects|Components|ParticleSystem|SetBoolParameter _fx_boost "isReady" false)
      (bind _returnvalue (Audio|SpawnSoundAttached (Variables|Sound|GetBoostSound) (Variables|Vehicle|GetSkeletalMesh)))
      (bind _output_get (Variables|Default|SetBoostAudioComponent _returnvalue))
      (Class|FortVehicleInterface|NotifyAudioComponentVehicleTriggered _output_get (Variables|Wraps|GetBoostSoundWrapPriority)))))

(event Custom|RefreshDamageState)

(event EventShowCooldownCue
  (params (Pawn :type Fort%20Player%20Pawn%20Object%20Reference) (Duration :type Float%20%28single-precision%29))
  ;[1 1 1 1 22] BOOST Events
  ;;[1 1 1] Boost Cooldown Cue
  (Ability|SendGameplayEventtoActor Pawn "(TagName=\"GameplayEffect.Vehicle.Boost\")" (Utilities|Struct|MakeGameplayEventData :ContextHandle "()" :InstigatorTags "(GameplayTags=)" :TargetTags "(GameplayTags=)" :EventMagnitude Duration :TargetData "()")))

(event EventOnBoostFinished
  ;[1 1 1 1 22] BOOST Events
  (Effects|Components|ParticleSystem|SetBoolParameter (Variables|Default|GetFX_Boost) "isActive" false)
  (Niagara|SpawnSystemAttached (Variables|Default|GetFX%20BoostEnd%20Template) (Variables|Vehicle|GetSkeletalMesh) "None" "0, 0, 0" "0, 0, 0" "KeepRelativeOffset" true true "None" false)
  (bind _returnvalue (Audio|SpawnSoundAttached (Variables|Sound|GetOnBoostEndSound) (Variables|Vehicle|GetSkeletalMesh)))
  (Class|FortVehicleInterface|NotifyAudioComponentVehicleTriggered _returnvalue (Variables|Wraps|GetBoostSoundWrapPriority))
  (Utilities|FlowControl|Delay 0.2)
  (if (Math|Float|NearlyEqual(Float) (Utilities|Struct|BreakFortRechargingActionTimer (Variables|Vehicle|GetBoostAction)) 1.0 0.01)
    _
    (else
      _))
  (Audio|SpawnSoundAttached "/Game/Athena/Sounds/Vehicles/Meatball/Meatball_Boost_Ready_Cue.Meatball_Boost_Ready_Cue" (Variables|Default|GetAudioController))
  (Effects|Components|ParticleSystem|SetBoolParameter (Variables|Default|GetFX_Boost) "isReady" true))

(event OnExitedWaterDelegate(FortMeatballPontoons)
  (params (Pontoon :type Spherical%20Pontoon%20Structure%20%28by%20ref%29))
  ;[1 1 1] EVENT -> On Pontoon Exit Water
  (switch name (Utilities|Struct|BreakSphericalPontoon Pontoon)
    (:RearPontoonBig
      (CallFunction|ExitedWaterMid))
    (:MidPontoonBig
      (CallFunction|ExitedWaterEngine))))

(event Custom|Check%20for%20Boost%20Ready
  ;[1 1 1 1 22] BOOST Events
  (Utilities|FlowControl|Delay 0.2)
  (if (Math|Float|NearlyEqual(Float) (Utilities|Struct|BreakFortRechargingActionTimer (Variables|Vehicle|GetBoostAction)) 1.0 0.01)
    _
    (else
      _))
  (Audio|SpawnSoundAttached "/Game/Athena/Sounds/Vehicles/Meatball/Meatball_Boost_Ready_Cue.Meatball_Boost_Ready_Cue" (Variables|Default|GetAudioController))
  (Effects|Components|ParticleSystem|SetBoolParameter (Variables|Default|GetFX_Boost) "isReady" true))

(event EventOnDamaged
  (params (Damage :type Float%20%28single-precision%29) (DamageTags :type Gameplay%20Tag%20Container%20Structure%20%28by%20ref%29) (EffectContext :type Gameplay%20Effect%20Context%20Handle%20Structure%20%28by%20ref%29) (EventInstigator :type Controller%20Object%20Reference) (DamageCauser :type Actor%20Object%20Reference))
  ;[1 1 1] EVENT -> On Damaged: Hit with Pickaxe
  (|CallParentFunction Damage DamageTags EffectContext EventInstigator DamageCauser)
  (CallFunction|DropLootonDamage DamageCauser (Ability|EffectContext|GetHitResult EffectContext))
  (Utilities|FlowControl|Sequence
    (:then_0
      (if (GameplayTags|HasTag DamageTags "(TagName=\"Weapon.Melee.Impact.Pickaxe\")")
        (if (> (Variables|Meatball|GetShouldPickaxeImpulse) 0.0)
          (Variables|Meatball|SetHitPickaxePawn DamageCauser)
          (CallFunction|HitwithPickaxe (Math|Vector|GetForwardVector (Pawn|GetControlRotation EventInstigator :!class /Script/Engine.Controller))))))
    (:then_1
      (Networking|SwitchHasAuthority
        (:Authority
          (bind _asfort_player_controller (Utilities|Casting|CastToFortPlayerController EventInstigator))
          (Variables|Default|SetLastDamagingPlayer _asfort_player_controller))))))

(event EventDestroyed
  (bind _lastdamagingplayer (Variables|Default|GetLastDamagingPlayer))
  (Networking|SwitchHasAuthority
    (:Authority
      (Utilities|IsValid _lastdamagingplayer
        (:Is%20Valid
          (Quest|SendComplexCustomStatEvent (Quests|GetQuestManager _lastdamagingplayer "Invalid") 0 "(GameplayTags=)" "(GameplayTags=((TagName=\"Athena.Quests.DestroyVehicle.Meatball\")))" 1))))))

(event EventOnBoostFailed
  (bind _returnvalue (Vehicle|GetDriver))
  ;[1 1 1 1 22] BOOST Events
  (Utilities|IsValid _returnvalue
    (:Is%20Valid
      (if (Pawn|IsLocallyControlled _returnvalue)
        (Audio|PlaySound2D (Variables|Sound|GetBoostFailedSound))))))

(event Custom|HitWithPickaxe
  (params (HitDirection :type Vector))
  (bind _returnvalue (Transformation|GetActorForwardVector))
  (bind (_pickaxeimpulse_x _pickaxeimpulse_y _pickaxeimpulse_z) (Math|Vector|BreakVector (Variables|GetPickaxeImpulse_X)))
  ;[1 1 1] EVENT -> On Damaged: Hit with Pickaxe
  (if (Vehicle|VehicleIsinAir)
    _
    (elif (and (<= (Math|Vector|DotProduct _returnvalue HitDirection) 0.0) (>= (Vehicle|GetPreviousForwardSpeedKmh) 2e+01))
      (Variables|Meatball|SetPickaxeImpulse (Math|Vector|vector*float (Math|Vector|MakeVector (.x _returnvalue) (.y _returnvalue) (Variables|Meatball|GetPickaxeForwardImpulseZ_Multiplier)) (Variables|Meatball|GetPickaxeImpulseStrength)))
      (if (not (CallFunction|IsinWaterBody))
        (Variables|Meatball|SetPickaxeImpulse (* (Variables|Meatball|GetPickaxeImpulse) "0.500000,0.500000,1.000000")))
      (Audio|BroadcastAbilityImpactatLocation 0 self (Math|Vector|MakeVector _pickaxeimpulse_x _pickaxeimpulse_y (Math|Float|Max(Float) _pickaxeimpulse_z 4e+02)) 777.0 "None" false)
      (else
        (Variables|Meatball|SetPickaxeImpulse (Math|Vector|vector*float (Math|Vector|MakeVector (.x HitDirection) (.y HitDirection) (select (>= (.z HitDirection) 0.0) (Math|Float|Clamp(Float) (.z HitDirection) 0.4 0.8) 0.3)) (Variables|Meatball|GetPickaxeImpulseStrength)))))))

(event Custom|Exited_Water_Mid
  ;[1 1 1] Exit / Enter Water MID
  (Niagara|SetNiagaraVariable(Bool) (Variables|FX|GetWaterEffectsComponent) "User.InWater" false) ; Turn off
)

(event EventOnDeathPlayEffects
  (params (Damage :type Float%20%28single-precision%29) (DamageTags :type Gameplay%20Tag%20Container%20Structure%20%28by%20ref%29) (Momentum :type Vector :default "0, 0, 0") (HitInfo :type Hit%20Result%20Structure%20%28by%20ref%29) (InstigatedBy :type Fort%20Pawn%20Object%20Reference) (DamageCauser :type Actor%20Object%20Reference) (EffectContext :type Gameplay%20Effect%20Context%20Handle%20Structure))
  (bind _returnvalue (Transformation|GetActorLocation))
  ;[1 1 1] EVENT -> On Death Play Effects
  (bind _returnvalue_2 (Effects|SpawnEmitteratLocation (Variables|FX|GetVehicleDestroyedFX) (Transformation|GetActorLocation) (Transformation|GetActorRotation) "1.000000,1.000000,1.000000" true "AutoRelease"))
  (Effects|Components|ParticleSystem|SetVectorParameter _returnvalue_2 "Velocity" (Transformation|GetVelocity))
  (Utilities|SpawnFNRadialForce (Transformation|GetWorldTransform (Variables|Vehicle|GetSkeletalMesh)) 0.3 0.875 2.25e+03 1.75
    (:Out
      (Camera|PlayWorldCameraShake (Variables|CameraShake|GetSmallJoltCameraShake) _returnvalue 1.5e+02 7e+02)
      (Audio|PlaySoundatLocation (Variables|Sound|GetOnDeathSound) _returnvalue)
      (Development|SetHiddeninGame (Variables|Vehicle|GetSkeletalMesh) true false)
      (Utilities|FlowControl|Sequence
        (:then_0
          (bind _damagedeffect (Variables|Default|GetDamagedEffect)
            (:then
              (Components|Activation|SetActive _damagedeffect false false))
            (:else)))
        (:then_1
          (bind _boostaudiocomponent (Variables|Default|GetBoostAudioComponent)
            (:then
              (Audio|Components|Audio|Stop _boostaudiocomponent))
            (:else)))))))

(event Custom|Entered_Water_Mid)

(event Custom|Entered_Water_Engine)

(event Custom|Exited_Water_Engine)

(event Custom|EventOnHealthChanged
  ;[1 1 1] Event On Health Changed
  (bind _output_get (Variables|VFX|SetDamageValue (Math|Float|Lerp 1.0 0.0 (/ (Class|FortVehicleInterface|GetHealth) (Class|FortVehicleInterface|GetMaxHealth)))))
  (CallFunction|UpdateDamageState _output_get))

(event Custom|DropLootOnExplode
  ;[1 1 1] Event -> Drop Loot On Explode
  (Class|AthenaFunctionLibrary|SpawnLootfromTable (Variables|DropLootOnDamage|GetLootOnDamagePackage) 1 -1 (Transformation|GetActorLocation) (Transformation|GetActorForwardVector) 0 true true))

(event Custom|UpdateDamageState
  (params (Damage :type Float%20%28double-precision%29))
  (bind _returnvalue (> Damage 0.75))
  (bind _damagedeffect (Variables|Default|GetDamagedEffect))
  ;[1 1 1] Update boat material based on remaining health %
  (Networking|SwitchHasAuthority
    (:Remote
      (CallFunction|SendDamageStatetoMaterial Damage)
      (if (> Damage 0.4)
        (Variables|Default|GetDamagedEffect
          (:then
            _)
          (:else
            (bind _returnvalue_2 (Effects|SpawnEmitterAttached :EmitterTemplate "/Game/Athena/DrivableVehicles/Meatball/Meatball_Large/FX/FX_Cascade/P_Meatball_Engine_Damage.P_Meatball_Engine_Damage" :AttachToComponent (Variables|Vehicle|GetSkeletalMesh) :AttachPointName "FX_EngineSocket" :Scale "1.000000,1.000000,1.000000" :LocationType "SnapToTarget" :bAutoDestroy true :PoolingMethod "None" :bAutoActivate true))
            (bind _output_get (Variables|Default|SetDamagedEffect _returnvalue_2))
            (EventDispatchers|BindEventtoOnParticleSpawn _output_get (delegate Custom|LightningSpark))))
        (Effects|Components|ParticleSystem|SetFloatParameter _damagedeffect "BadlyDamaged" (Math|Conversions|ToFloat(Boolean) _returnvalue))
        (Effects|Components|ParticleSystem|SetFloatParameter _damagedeffect "Ribbon" (Math|Conversions|ToFloat(Boolean) _returnvalue))
        (Effects|Components|ParticleSystem|SetFloatParameter _damagedeffect "NearlyDestroyed" (Math|Conversions|ToFloat(Boolean) (> Damage 0.9)))
        (else
          (Variables|Default|GetDamagedEffect
            (:then
              (Components|DestroyComponent (Variables|Default|GetDamagedEffect)))
            (:else)))))))

(event Custom|LightningSpark
  (params (EventName :type Name) (EmitterTime :type Float%20%28single-precision%29) (Location :type Vector) (Velocity :type Vector))
  ;[1 1 1] Update boat material based on remaining health %
  (if (Utilities|Name|Equal(Name) EventName "spark")
    (Audio|Components|Audio|Play (Variables|Default|GetAudioSpark) 0.0)))

(event OnComponentHit(SkeletalMesh)
  (params (HitComponent :type Primitive%20Component%20Object%20Reference) (OtherActor :type Actor%20Object%20Reference) (OtherComp :type Primitive%20Component%20Object%20Reference) (NormalImpulse :type Vector :default "0, 0, 0") (Hit :type Hit%20Result%20Structure%20%28by%20ref%29))
  (bind _returnvalue (Vehicle|GetPreviousForwardSpeedKmh))
  (bind (_returnvalue_x _returnvalue_y _returnvalue_z) (Math|Vector|BreakVector (* (Transformation|GetActorForwardVector) (* _returnvalue (Variables|Meatball|GetSpeedImpulsePlayerMulti)))))
  (bind _returnvalue_2 (Vehicle|GetPreviousForwardSpeedKmh))
  (bind (_returnvalue_x_2 _returnvalue_y_2 _returnvalue_z_2) (Math|Vector|BreakVector (* (Transformation|GetActorForwardVector) (* _returnvalue_2 (Variables|Meatball|GetSpeedImpulsePlayerMulti)))))
  (bind _minspeedtolaunchplayer (Variables|Meatball|GetMinSpeedToLaunchPlayer))
  ;[1 1 1] Launch Player
  (if (> (Variables|Meatball|GetShouldLaunchPlayer) 0.0)
    (if (Utilities|NotEqual(Object) (Variables|Default|GetLastDriver) OtherActor)
      (bind _asfort_player_pawn_athena (Utilities|Casting|CastToFortPlayerPawnAthena OtherActor)
        (:then
          (if (Parachute|IsSkydiving _asfort_player_pawn_athena)
            _
            (elif (or (Utilities|Array|ContainsItem (Variables|Meatball|GetPassengerPawns) _asfort_player_pawn_athena) (Utilities|Equal(Object) _asfort_player_pawn_athena (Variables|Meatball|GetLastHitPlayer)))
              _
              (elif (>= _returnvalue_2 _minspeedtolaunchplayer)
                (Character|LaunchCharacter _asfort_player_pawn_athena (Math|Vector|MakeVector _returnvalue_x_2 _returnvalue_y_2 (Math|Float|Max(Float) _returnvalue_z_2 1.5e+03)) true true) ; Launch player based on cart velocity
                (for _array_element (Vehicle|GetAllPassengers)
                  (GameplayCue|ExecuteGameplayCue%28Burst%29Local _array_element (Variables|GameplayCues|GetGC_HitPlayer) (Variables|GameplayCues|GetGC_ParamsEmpty))
                  (Variables|Meatball|SetLastHitPlayer _asfort_player_pawn_athena)
                  (Audio|OnCollision (Variables|Default|GetFortCollisionAudioHitPlayer) (Transformation|GetActorLocation) (Math|Vector|MakeVector 0.0 0.0 8e+04))
                  (if (Vehicle|HasDriver)
                    (Class|FortPlayerPawnAthena|SetFallInstigator :self _asfort_player_pawn_athena :NewFallInstigator (Game|GetInstigatorController (Variables|Meatball|GetDriverPawn)))
                    (Utilities|FlowControl|Delay 0.2)
                    (Variables|Meatball|SetLastHitPlayer 0)
                    (Variables|Meatball|SetHitPickaxePawn 0)))))))
        (:CastFailed
          (bind _asfort_aipawn (Utilities|Casting|CastToFortAIPawn OtherActor))
          (Components|Movement|StopMovementImmediately (Variables|/Script/Engine.Character|GetCharacterMovement _asfort_aipawn))
          (if (>= _returnvalue _minspeedtolaunchplayer)
            (Character|LaunchCharacter _asfort_aipawn (Math|Vector|MakeVector _returnvalue_x _returnvalue_y (Math|Float|Max(Float) _returnvalue_z 1e+03)) true true) ; Launch player based on cart velocity
            (Networking|SwitchHasAuthority
              (:Remote
                (for _array_element_2 (Vehicle|GetAllPassengers)
                  (GameplayCue|ExecuteGameplayCue%28Burst%29Local _array_element_2 (Variables|GameplayCues|GetGC_HitFiend) (Variables|GameplayCues|GetGC_ParamsEmpty)))))))))))

(event Custom|DropLootOnDamage
  (params (DamagingActor :type Actor%20Object%20Reference) (HitResult :type Hit%20Result%20Structure))
  (bind _returnvalue (Class|FortVehicleInterface|GetHealth))
  (bind _returnvalue_2 (Class|FortVehicleInterface|GetMaxHealth))
  (bind _returnvalue_3 (Math|Float|Max(Float) _returnvalue_2 0.1))
  (bind _returnvalue_4 (/ _returnvalue _returnvalue_3))
  (bind _row_lootdropdamagethresholds (Variables|DropLootonDamage|GetRow_LootDropDamageThresholds))
  (bind _returnvalue_5 (ScalableFloat|GetValueatLevel _row_lootdropdamagethresholds))
  (bind _returnvalue_6 (Math|Float|Max(Float) _returnvalue_5 0.1))
  (bind _returnvalue_7 (/ _returnvalue_4 _returnvalue_6))
  (bind _returnvalue_8 (* _returnvalue_7 1e+02))
  (bind _returnvalue_9 (Math|Float|Floor _returnvalue_8))
  (bind (_bblockinghit _binitialoverlap _time _distance _location _impactpoint _normal) (Collision|BreakHitResult HitResult))
  ;[1 1 1] Event -> Drop Loot On Damage
  (if (Class|AthenaFunctionLibrary|RowtoBool (Variables|DropLootonDamage|GetRow_ShouldDropLootOnDamage))
    (if (< _returnvalue_9 (Variables|DropLootonDamage|GetLootDropCurentThreshold))
      (bind _returnvalue_10 (TargetSelection|PushOffHitLocation self HitResult (Variables|DropLootonDamage|GetLootTossOffset)))
      (bind _itempickupsspawned (Class|AthenaFunctionLibrary|SpawnLootfromTable (Variables|DropLootOnDamage|GetLootOnDamagePackage) 1 -1 _returnvalue_10 _normal 0 true true)) ; Roll a loot package and toss it in the direction of the player
      (Effects|SpawnEmitteratLocation (Variables|Default|GetFX_LootDropOnDamagedSpawn) _returnvalue_10 (Math|Conversions|RotationFromXVector _normal) "1.000000,1.000000,1.000000" true "AutoRelease")
      (for _array_element _itempickupsspawned
        (Pickup|TossPickup _array_element (Math|Vector|SelectVector (+ _location (* (Math|Vector|Normalize (- (Transformation|GetActorLocation DamagingActor) _location)) (Variables|DropLootOnDamage|GetLootTossMin))) (Transformation|GetActorLocation DamagingActor) (>= (Math|Vector|VectorLength (- (Transformation|GetActorLocation DamagingActor) _location)) (Variables|DropLootOnDamage|GetLootTossMin))) 0 0 true false "Tossed" "Unset"))
      (else
        _))
    (Variables|DropLootonDamage|SetLootDropCurentThreshold _returnvalue_9) ; Update the current threshold so that it can be used for the next check
  ))

(event Vehicle|Water|EventOnEnteredWaterVolume
  (params (WaterSurfacePoint :type Vector%20%28by%20ref%29 :default "0, 0, 0"))
  (bind _returnvalue (Transformation|GetActorLocation))
  (bind _skeletalmesh (Variables|Vehicle|GetSkeletalMesh))
  ;[1 1 1] EVENT -> On Entered Water Volume -- FX for water ineraction
  (Networking|CanExecuteCosmeticEvents
    (:True
      (if (> (Rendering|GetRenderingDetailMode) 0)
        (bind _returnvalue_2 (Utilities|Time|SetTimerbyEvent (delegate Custom|CheckWater) 0.25 true))
        (Variables|Meatball|SetCheckWaterTimer _returnvalue_2)
        (bind _inwaterfx (Variables|Default|GetInWaterFX)
          (:then
            (Components|DestroyComponent _inwaterfx))
          (:else
            _))
        (bind _returnvalue_3 (Effects|SpawnEmitterAttached (Variables|FX|GetInWaterLoopingFX) _skeletalmesh "Jet" _returnvalue)) ; Looping water FX
        (Variables|Default|SetInWaterFX _returnvalue_3)
        (Effects|SpawnEmitterAttached (Variables|FX|GetWaterSplashBurstFX) _skeletalmesh "Jet" _returnvalue) ; Splash burst FX
        (bind _inwaterloop (Variables|Default|GetInWaterLoop)
          (:then
            (Audio|Components|Audio|Stop _inwaterloop))
          (:else
            _))
        (Audio|SpawnSoundAttached (Variables|Sound|GetWaterEnterSound) (Variables|Vehicle|GetSkeletalMesh))
        (bind _returnvalue_4 (Audio|SpawnSoundAttached (Variables|Sound|GetInWaterLoopSound) (Variables|Vehicle|GetSkeletalMesh)))
        (Variables|Default|SetInWaterLoop _returnvalue_4)))))

(event Custom|DisableWeapon
  (params (DisableWeapon :type Boolean))
  ;[1 0.7332 0.9615 0.8] Disable if set (after spawn)
  (if DisableWeapon
    (bind _fortvehicleseatweapon (Variables|Default|GetFortVehicleSeatWeapon)
      (:then
        (Components|DestroyComponent _fortvehicleseatweapon))
      (:else))))

(event Custom|CheckWater
  ;[1 1 1] EVENT -> On Entered Water Volume -- FX for water ineraction
  (Effects|Components|ParticleSystem|SetVectorParameter (Variables|Default|GetInWaterFX) "Offset" (Math|Vector|Normalize (Math|Transform|InverseTransformDirection (Transformation|GetWorldTransform (Variables|Default|GetInWaterFX)) (Transformation|GetVelocity))))
  (bind _inwaterloop (Variables|Default|GetInWaterLoop)
    (:then
      (Audio|Components|Audio|SetVolumeMultiplier _inwaterloop (Math|Float|MapRangeClamped (Vehicle|GetForwardSpeedKmh) 0.0 2e+01 0.0 1.0) :!class /Script/Engine.AudioComponent)
      (bind _inwaterfx (Variables|Default|GetInWaterFX)
        (:then
          (Components|Activation|SetActive _inwaterfx true false))
        (:else)))
    (:else)))

(event OnVehicleSeatPlayerSlotsChanged(VehicleSeatComponent)
  (bind _fx_boost (Variables|Default|GetFX_Boost))
  (bind _fx_boost_2 (Variables|Default|GetFX_Boost))
  ;[0 0 0] FORT-782439 optimize to disable fx on no driver
  (if (Vehicle|HasDriver)
    (if (not (Components|Activation|IsActive _fx_boost))
      (Components|Activation|Activate _fx_boost true)
      (Effects|Components|ParticleSystem|SetBoolParameter (Variables|Default|GetFX_Boost) "isReady" (not (< (CallFunction|GetBoostCharge) 1.0)))
      (CallFunction|CheckForBoostReady)
      (CallFunction|PlayEngineonSound))
    (else
      (Effects|Components|ParticleSystem|SetBoolParameter _fx_boost_2 "isReady" false)
      (Components|Activation|Deactivate _fx_boost_2)
      (CallFunction|PlayEngineOffSound))))

(event Vehicle|Water|EventOnExitedWaterVolume
  (params (WaterSurfacePoint :type Vector%20%28by%20ref%29 :default "0, 0, 0"))
  ;[1 1 1] EVENT -> On Exit Water Volume
  (Networking|CanExecuteCosmeticEvents
    (:True
      (Utilities|Time|ClearandInvalidateTimerbyHandle (Variables|Meatball|GetCheckWaterTimer))
      (bind _inwaterfx (Variables|Default|GetInWaterFX)
        (:then
          (Components|Activation|SetActive _inwaterfx false false)
          (bind _inwaterloop (Variables|Default|GetInWaterLoop)
            (:then
              (Audio|Components|Audio|FadeOut _inwaterloop 3.0 0.0 "Linear" :!class /Script/Engine.AudioComponent))
            (:else)))
        (:else)))))

(event Custom|PlayEngineOnSound
  (if (not (Vehicle|IsDisabled))
    (Networking|CanExecuteCosmeticEvents
      (:True
        (if (not (Variables|Default|GetbEngineAudio))
          (bind _returnvalue (Audio|SpawnSoundAttached (Variables|Sound|GetEngineOnSound) (Variables|Default|GetAudioController)))
          (Audio|Parameter|SetIntegerParameter _returnvalue "IsInWater" (Math|Conversions|ToInteger(Boolean) (CallFunction|IsinWaterBody)) :!class /Script/AudioExtensions.AudioParameterControllerInterface)
          (Variables|Default|SetbEngineAudio true))))))

(event EventTick
  (params (DeltaSeconds :type Float%20%28single-precision%29))
  (bind _set_up_snow_reference (Variables|Default|GetSet%20up%20snow%20reference))
  ;[1 0.94 0.4479] EVENT -> Tick:    We need these values in order for the niagara system to function correctly.
  (Utilities|IsValid _set_up_snow_reference
    (:Is%20Valid
      (Variables|FortMeatballVehicle|SetbHasSnow true)
      (Variables|FortMeatballVehicle|SetCurrentSnowAltitude (Variables|/Game/Athena/Apollo/Environments/Blueprints/CalendarEvents/BP_ApolloSnowSetup.BP_ApolloSnowSetup_C|GetCurrentSnowAltitude _set_up_snow_reference)))
    (:Is%20Not%20Valid
      (Variables|FortMeatballVehicle|SetbHasSnow false))))

(event Custom|PlayEngineOffSound
  (Networking|CanExecuteCosmeticEvents
    (:True
      (if (Variables|Default|GetbEngineAudio)
        (bind _returnvalue (Audio|SpawnSoundAttached (Variables|Sound|GetEngineOffSound) (Variables|Default|GetAudioController)))
        (Audio|Parameter|SetIntegerParameter _returnvalue "IsInWater" (Math|Conversions|ToInteger(Boolean) (CallFunction|IsinWaterBody)) :!class /Script/AudioExtensions.AudioParameterControllerInterface)
        (Variables|Default|SetbEngineAudio false)))))

(event Custom|EventOnHandleOutOfHealth
  (params (DeadVehicle :type Fort%20Athena%20Vehicle%20Object%20Reference) (InTags :type Gameplay%20Tag%20Container%20Structure%20%28by%20ref%29) (EventInstigator :type Controller%20Object%20Reference) (DamageCauser :type Actor%20Object%20Reference))
  (bind (_origin _boxextent) (Collision|GetComponentBounds (Class|FortVehicleInterface|GetSkeletalMeshComponent)))
  (if (Networking|IsServer)
    (Utilities|FlowControl|DoOnce false
      (:Completed
        (if (ScalableFloat|GetValueasBool (Variables|Default|GetCurieHotfix_FireApplication))
          (bind _outhits (Collision|MultiSphereTraceForObjects (Transformation|GetActorLocation) (+ (Transformation|GetActorLocation) (Math|Vector|GetUpVector "0, 0, 0")) (Variables|Default|GetExplodeRadius) (Variables|Default|GetObjectArray)))
          (for _array_element _outhits
            (Curie|ExecuteFireApplication EventInstigator DamageCauser (.HitActor (Collision|BreakHitResult _array_element)) 1e+02 _array_element))
          (Curie|IgniteGrassinBounds EventInstigator DamageCauser (Math|Box|MakeBox (Utilities|Operators|Subtract _origin _boxextent (Math|Vector|MakeVector 64.0 64.0 64.0)) (Utilities|Operators|Add _origin _boxextent (Math|Vector|MakeVector 64.0 64.0 64.0)))))))))

(event Custom|StopTurnRumbleBP
  ;[1 1 1] Turn Rumble
  (Utilities|FlowControl|Sequence
    (:then_0
      _)
    (:then_1
      (Utilities|FlowControl|DoOnce false
        (:Completed
          _))))
  (Game|Feedback|PlayDynamicForceFeedback (Pawn|GetLocalViewingPlayerController (Variables|Meatball|GetDriverPawn)) (Variables|Default|GetTurnIntensity) -1 false (not (> (Vehicle|GetSteeringAngle) 0.0)) false (> (Vehicle|GetSteeringAngle) 0.0)))

(event Custom|StartTurnRumble
  ;[1 1 1] Turn Rumble
  (Networking|CanExecuteCosmeticEvents
    (:True
      (if (Pawn|IsLocallyControlled (Variables|Meatball|GetDriverPawn))
        (Utilities|FlowControl|Sequence
          (:then_0
            (Utilities|FlowControl|DoOnce false
              (:Completed
                _)))
          (:then_1
            _))
        (Game|Feedback|PlayDynamicForceFeedback (Pawn|GetLocalViewingPlayerController (Variables|Meatball|GetDriverPawn)) (Variables|Default|GetTurnIntensity) -1 false (not (> (Vehicle|GetSteeringAngle) 0.0)) false (> (Vehicle|GetSteeringAngle) 0.0))))))

(event Vehicle|FX|EventOnSetVehicleFx
  (params (FxTag :type Gameplay%20Tag%20Structure) (FxSystem :type FXSystem%20Asset%20Object%20Reference))
  ;[1 1 1] OnSetVehicleFx
  (switch Utilities|FlowControl|Switch|SwitchonGameplayTag FxTag
    (:Athena.Vehicle.Meatball.FX.Boost
      (bind _asniagara_system (Utilities|Casting|CastToNiagaraSystem FxSystem))
      (Niagara|SetNiagaraSystemAsset (Variables|Default|GetFX_Boost) _asniagara_system true))
    (:Athena.Vehicle.Meatball.FX.GroundInteraction
      (bind _asniagara_system_2 (Utilities|Casting|CastToNiagaraSystem FxSystem))
      (Niagara|SetNiagaraSystemAsset (Variables|Default|GetFX_GroundInteraction) _asniagara_system_2 true))
    (:Athena.Vehicle.Meatball.FX.LargeImpact
      (bind _asniagara_system_3 (Utilities|Casting|CastToNiagaraSystem FxSystem))
      (Variables|Default|SetFX%20Large%20Impact _asniagara_system_3))))

(event SurfaceType|EventOnSurfaceTypeVehicleIsOnChanged
  (params (SurfaceTypeVehicleIsOn :type EPhysicalSurface%20Enum :default "SurfaceType_Default"))
  (Default|HandleSurfaceTypeChanged SurfaceTypeVehicleIsOn))

(fn calculateTurnSpeed
  (return (select (> (Math|Vector|DotProduct (* (Math|Vector|Normalize (Transformation|GetVelocity self)) "1.000000,1.000000,0.000000") (* (Math|Vector|Normalize (Transformation|GetRightVector (Variables|Vehicle|GetSkeletalMesh))) "1.000000,1.000000,0.000000")) 0.0) 1.0 -1.0) (/ (Math|Float|Absolute(Float) (Math|Vector|DotProduct (* (Math|Vector|Normalize (Transformation|GetVelocity self)) "1.000000,1.000000,0.000000") (* (Math|Vector|Normalize (Transformation|GetRightVector (Variables|Vehicle|GetSkeletalMesh))) "1.000000,1.000000,0.000000"))) 1.0)))

(fn CanDamageBuilding
  (params (Hit :type Hit%20Result%20Structure%20%28by%20ref%29))
  (bind (_bblockinghit _binitialoverlap _time _distance _location _impactpoint _normal) (Collision|BreakHitResult Hit))
  (return (Math|Boolean|ANDBoolean (Variables|/Script/Engine.Actor|GetbCanBeDamaged (Variables|Meatball|GetHitBuildingActor)) (Class|FortVehicleInterface|HasPermissiontoAlterBuildings) (<= (Health|GetMaxHealth (Variables|Meatball|GetHitBuildingActor) :!class /Script/FortniteGame.BuildingActor) 3e+03) (and (<= (Math|Vector|DotProduct (Math|Vector|Normalize (Vehicle|GetPreviousLinearVelocity)) _normal) (Variables|Meatball|GetDestructionAngle)) (>= (Vehicle|GetPreviousForwardSpeedKmh) (Variables|Meatball|GetMinSpeedToDamage))) (not (Actor|ActorHasTag (Variables|Meatball|GetHitBuildingActor) "NoVehicleDamage")))))

(fn DetermineWaterMovementForLazyUpdateMesh
  (params (Forward%20Dot :type Float%20%28double-precision%29) (Right%20Dot :type Float%20%28double-precision%29))
  (return (+ (* (Math|Vector|Lerp(Vector) "1.000000,0.000000,0.000000" "-1.000000,1.000000, 0" (Math|Float|MapRangeClamped Forward%20Dot -1.0 1.0 0.0 1.0)) (Variables|VFX|GetForward_Intensity)) (* (Math|Vector|Lerp(Vector) "0.000000,-1.000000,0.000000" "0.000000,1.000000,0.000000" (Math|Float|MapRangeClamped Right%20Dot -1.0 1.0 0.0 1.0)) (Variables|VFX|GetRight_Intensity)))))

(fn HandleSurfaceTypeChanged
  (params (SurfaceType :type EPhysicalSurface%20Enum :default "SurfaceType_Default"))
  (if (Utilities|Enum|Equal(Enum) SurfaceType "SurfaceType24")
    (GameplayEffects|ApplyGameplayEffectToSelf :self (Vehicle|GetVehicleAbilitySystemComponent) :GameplayEffectClass "/Motorboat/Vehicle/Abilities/GE_Meatball_Surface_Ice.GE_Meatball_Surface_Ice_C" :Level 0.0)
    (else
      (GameplayEffects|RemoveActiveGameplayEffectbySourceEffect (Vehicle|GetVehicleAbilitySystemComponent) "/Motorboat/Vehicle/Abilities/GE_Meatball_Surface_Ice.GE_Meatball_Surface_Ice_C" 0 -1))))

(fn IsNiagaraEffectsEnabled
  (Utilities|IsValid 0
    (:Is%20Valid
      (return (Effects|IsNiagaraEffectEnabled "Vehicle" (Niagara|GetNiagaraSystemAsset (Variables|FX|GetWaterEffectsComponent)))))
    (:Is%20Not%20Valid
      (return false))))

(fn IsOnGround
  (return (not (or (CallFunction|IsinWaterBody) (Vehicle|VehicleIsinAir)))))

(fn ManageStateDriverEnter
  (params (DriverPawn :type Fort%20Player%20Pawn%20Object%20Reference))
  (bind _output_get (Variables|Meatball|SetDriverPawn DriverPawn))
  (EventDispatchers|BindEventtoOnFiringRangedWeapon _output_get (EventDispatchers|CreateEvent :!event FireFF))
  (Utilities|FlowControl|Sequence
    (:then_0
      ;[1 1 1] Only true on local driving client and server
      (bind _asplayer_controller (Utilities|Casting|CastToPlayerController (Pawn|GetController _output_get)))
      (Variables|Meatball|SetBP%20Driving%20Player%20Controller _asplayer_controller)
      (CallFunction|LandDamageEvent))
    (:then_1
      (GameplayEffects|ApplyGameplayEffectToTarget :self (Ability|GetAbilitySystemComponent _output_get) :GameplayEffectClass (Variables|Quests|GetDriverGE) :Target (Ability|GetAbilitySystemComponent _output_get) :Level 0.0)
      (Networking|CanExecuteCosmeticEvents
        (:True
          (if (not (Vehicle|IsDisabled))
            (CallFunction|PlayEngineonSound)
            (CallFunction|CheckForBoostReady)))))
    (:then_2
      (CallFunction|PlayStartupFF))
    (:then_3
      (if (not (CallFunction|IsinWaterBody))
        (CallFunction|ExitedWaterMid)))))

(fn ManageStateDriverExit
  (params (PlayerPawn :type Fort%20Player%20Pawn%20Object%20Reference))
  (bind _bp_driving_player_controller (Variables|Meatball|GetBP%20Driving%20Player%20Controller))
  (EventDispatchers|UnbindEventfromOnFiringRangedWeapon PlayerPawn (EventDispatchers|CreateEvent :!event FireFF))
  (Utilities|FlowControl|Sequence
    (:then_0
      ;[1 1 1] Clear Camera Shake and Rumble
      (CallFunction|BoostRumbleEnd)
      (Utilities|Time|ClearandInvalidateTimerbyHandle (Variables|Meatball|GetDamageOnLandTimer))
      (Utilities|Time|ClearandInvalidateTimerbyHandle (Variables|Meatball|GetLandDamageTimer))
      (CallFunction|StopStartupRumble)
      (Utilities|IsValid _bp_driving_player_controller
        (:Is%20Valid
          (Game|Feedback|ClientStopCameraShake _bp_driving_player_controller (Variables|Default|GetOnHitShake) true)
          (Game|Feedback|ClientStopCameraShake _bp_driving_player_controller (Variables|Default|GetOnBigHitShake) true))))
    (:then_1
      ;[1 1 1] Remove GE for quest
      (if (not (FortPawn|IsDead PlayerPawn))
        (GameplayEffects|RemoveActiveGameplayEffectbySourceEffect (Ability|GetAbilitySystemComponent PlayerPawn) (Variables|Quests|GetDriverGE) 0 -1)))
    (:then_2
      ;[1 1 1] Clear after using to stop camera shakes/rumble
      (Variables|Meatball|SetBP%20Driving%20Player%20Controller 0)
      (Variables|Meatball|SetDriverPawn 0)
      (Variables|Default|SetLastDriver PlayerPawn))
    (:then_3
      (Networking|CanExecuteCosmeticEvents
        (:True
          (CallFunction|PlayEngineOffSound)
          (Effects|Components|ParticleSystem|SetBoolParameter (Variables|Default|GetFX_Boost) "isReady" false) ; Deactivate boost ready FX on driver exit FORT-502080
        )))))

(fn ManageStatePassengerEnter
  (params (PassengerPawn :type Fort%20Player%20Pawn%20Object%20Reference))
  (bind _output_get (Variables|Meatball|SetPassengerPawn PassengerPawn))
  (Utilities|FlowControl|Sequence
    (:then_0
      ;[1 1 1] Cache Passenger Player Controller
      (bind _asplayer_controller (Utilities|Casting|CastToPlayerController (Pawn|GetController (Variables|Meatball|GetPassengerPawn))))
      (bind _output_get_2 (Variables|Meatball|SetBP%20PassengerPlayerController _asplayer_controller)))
    (:then_1
      (GameplayEffects|ApplyGameplayEffectToTarget :self (Ability|GetAbilitySystemComponent _output_get) :GameplayEffectClass (Variables|Quests|GetPassengerGE) :Target (Ability|GetAbilitySystemComponent _output_get) :Level 0.0))))

(fn ManageStatePassengerExit
  (params (PlayerPawn :type Fort%20Player%20Pawn%20Object%20Reference))
  (bind _bp_passengerplayercontroller (Variables|Meatball|GetBP%20PassengerPlayerController))
  (Utilities|FlowControl|Sequence
    (:then_0
      ;[1 1 1] Stop Passenger Camera Shake
      (Utilities|IsValid _bp_passengerplayercontroller
        (:Is%20Valid
          (Game|Feedback|ClientStopCameraShake _bp_passengerplayercontroller (Variables|Default|GetOnHitShake) true)
          (Game|Feedback|ClientStopCameraShake _bp_passengerplayercontroller (Variables|Default|GetOnBigHitShake) true)
          (Utilities|IsValid (Variables|Components|GetPassengerCameraShake)
            (:Is%20Valid
              (CameraShakes|StopCameraShake (Variables|/Script/Engine.PlayerController|GetPlayerCameraManager (Variables|Meatball|GetBP%20PassengerPlayerController)) (Variables|Components|GetPassengerCameraShake) false)
              (Variables|Components|SetPassengerCameraShake 0))))))
    (:then_1
      ;[1 1 1] Remove GE for quest
      (if (not (FortPawn|IsDead PlayerPawn))
        (GameplayEffects|RemoveActiveGameplayEffectbySourceEffect (Ability|GetAbilitySystemComponent PlayerPawn) (Variables|Quests|GetPassengerGE) 0 -1)))
    (:then_2
      ;[1 1 1] Clear refs
      (Variables|Meatball|SetBP%20PassengerPlayerController 0)
      (Variables|Meatball|SetPassengerPawn 0))))

(fn NiagaraUserVariables
  ;[1 1 1] All the user varaibles are set behind the scenes in code. https://p4-swarm.epicgames.net/changes/9102662
  (Niagara|SetNiagaraVariable(Float) (Variables|FX|GetWaterEffectsComponent) "User.Boost" (Variables|Meatball|GetBoost)))

(fn OnRep_IsLegendary)

(fn OnRep_LandDamageFrequency)

(fn PlayHapticsOnHit
  (params (HitMagnitude :type Float%20%28double-precision%29))
  (bind _bp_passengerplayercontroller (Variables|Meatball|GetBP%20PassengerPlayerController))
  (bind _bp_driving_player_controller (Variables|Meatball|GetBP%20Driving%20Player%20Controller))
  (bind _validplayercontrollers (LocalVariables|GetValidPlayerControllers))
  (Utilities|FlowControl|Sequence
    (:then_0
      (Utilities|IsValid _bp_driving_player_controller
        (:Is%20Valid
          (Utilities|Array|Add (LocalVariables|GetValidPlayerControllers) _bp_driving_player_controller))))
    (:then_1
      (Utilities|IsValid _bp_passengerplayercontroller
        (:Is%20Valid
          (Utilities|Array|Add (LocalVariables|GetValidPlayerControllers) _bp_passengerplayercontroller))))
    (:then_2
      (if (> (Utilities|Array|Length (LocalVariables|GetValidPlayerControllers)) 0)
        (if (> (Variables|GetHitMagnitude) 1e+04)
          (Utilities|FlowControl|Sequence
            (:then_0
              (LocalVariables|SetCameraShakeMagnitude (Math|Float|MapRangeClamped (Variables|GetHitMagnitude) 1.5e+04 5e+04 0.1 1.0))
              (Utilities|IsValid (Variables|Meatball|GetBP%20Driving%20Player%20Controller)
                (:Is%20Valid
                  (Game|Feedback|ClientStartCameraShake :self (LocalVariables|GetValidPlayerControllers) :Shake (Variables|Default|GetOnHitShake) :Scale (LocalVariables|GetCameraShakeMagnitude) :PlaySpace "CameraLocal")
                  (if (> (Variables|GetHitMagnitude) 2e+04)
                    (Game|Feedback|ClientPlayForceFeedback _validplayercontrollers "/Game/Weapons/ForceFeedback/FF_Generic_Impact_Near.FF_Generic_Impact_Near" "None" false false false)
                    (else
                      (Game|Feedback|ClientPlayForceFeedback _validplayercontrollers "/Game/Weapons/ForceFeedback/FF_Generic_Build.FF_Generic_Build" "None" false false false))))))
            (:then_1
              (if (< (Math|Float|Absolute(Float) (Math|Vector|DotProduct (Math|Vector|Normalize (Vehicle|GetPreviousLinearVelocity)) "0, 0,1.000000")) 0.5)
                (bind _output_get (LocalVariables|SetHitDelta (- (Variables|Vehicle|GetCachedSpeed) (Math|Vector|VectorLength (Transformation|GetComponentVelocity (Variables|Vehicle|GetSkeletalMesh))))))
                (if (> _output_get (LocalVariables|GetCameraPushVelocityDelta))
                  (Game|Feedback|ClientStartCameraShake (LocalVariables|GetValidPlayerControllers) (Variables|Default|GetOnBigHitShake) (Math|Float|MapRangeClamped (LocalVariables|GetHitDelta) 3e+02 9e+02 0.5 1.0) "UserDefined" (Math|Rotator|MakeRotfromX (Vehicle|GetPreviousLinearVelocity))))))))))))

(fn SendDamageStateToMaterial
  (params (Param%20Value :type Float%20%28double-precision%29))
  (bind _output_get (Variables|FX|SetSavedDamageForMID Param%20Value))
  (Rendering|Material|SetScalarParameterValueonMaterials (Variables|Vehicle|GetSkeletalMesh) "Damage" _output_get))

(fn UpdateDynamicCameraShake)

(fn ConstructionScript
  (Variables|FortAthenaVehicle|SetOverlapComponent (Variables|Default|GetOverlapVolume))
  (Variables|Meatball|SetShouldPickaxeImpulse (ScalableFloat|GetValueatLevel (Variables|HotfixLookup|GetRow_ShouldPickaxeImpulse)))
  (Variables|Meatball|SetShouldLaunchPlayer (ScalableFloat|GetValueatLevel (Variables|HotfixLookup|GetRow_ShouldLaunchPlayer)))
  (Variables|Meatball|SetPickaxeImpulseStrength (ScalableFloat|GetValueatLevel (Variables|HotfixLookup|GetRow_PickaxeImpulseStrength)))
  (Variables|Meatball|SetMinSpeedToLaunchPlayer (ScalableFloat|GetValueatLevel (Variables|HotfixLookup|GetRow_MinSpeedLaunchPlayer)))
  (Variables|Meatball|SetSpeedImpulsePlayerMulti (ScalableFloat|GetValueatLevel (Variables|HotfixLookup|GetRow_SpeedImpulseMultiplier)))
  (Variables|Meatball|SetMinSpeedToDamage (ScalableFloat|GetValueatLevel (Variables|HotfixLookup|GetRow_MinSpeedToDamage)))
  (Variables|Meatball|SetPickaxeForwardImpulseZ_Multiplier (ScalableFloat|GetValueatLevel (Variables|HotfixLookup|GetRow_PickaxeImpulseZ_Multiplier)))
  (Variables|Meatball|SetShouldDamageOnLand (ScalableFloat|GetValueatLevel (Variables|HotfixLookup|GetRow_ShouldDamageOnLand)))
  (Variables|Meatball|SetBoatDamageOnLand (ScalableFloat|GetValueatLevel (Variables|HotfixLookup|GetRow_BoatDamageOnLand))))
